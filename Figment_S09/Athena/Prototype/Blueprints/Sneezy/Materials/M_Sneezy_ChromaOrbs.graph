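# Additive, unlit particle sprite material that draws glowing orbs with chromatic-aberration color fringes, procedurally and without textures.
# Three soft sphere masks supply green at the UV center and red and blue at equal, opposite offsets, with the offset varied by per-particle random.
# The summed channels are raised to an Exponent parameter to shape the falloff, then tinted by particle color.
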Material /Figment_S09/Athena/Prototype/Blueprints/Sneezy/Materials/M_Sneezy_ChromaOrbs
graph {
 properties {bUsedWithParticleSprites=true, blendMode=BLEND_Additive, pixelDepthOffsetMode=PDOM_Legacy, shadingModel=MSM_Unlit, usedShadingModels=Unlit}
 nodes [
  n1 = TextureCoordinate()
  n2 = Constant2Vector() {g=0.5, r=0.5}
  n3 = Constant() {r=0.4}
  n4 = Constant()
  n5 = LinearInterpolate(A=Constant2Vector() {g=0.02, r=0.1}, B=Constant2Vector() {r=0.025}, Alpha=ParticleRandom()) {constB=0}
  n6 = Add(A=@n2, B=@n5)
  n7 = Add(
   A=Multiply(A=SphereMask(A=@n1, B=@n2, Radius=@n3, Hardness=@n4), B=Constant3Vector() {constant=<rgba 0 1 0 0>})
   B=Multiply(A=SphereMask(A=@n1, B=@n6, Radius=@n3, Hardness=@n4), B=Constant3Vector() {constant=<rgba 1 0 0 0>})
  )
  n8 = Multiply(A=@n5) {constB=-1}
  n9 = Multiply(A=SphereMask(A=@n1, B=Add(A=@n2, B=@n8) {constB=0}, Radius=@n3, Hardness=@n4), B=Constant3Vector() {constant=<rgba 0 0 1 0>})
  Multiply(A=Power(Base=Add(A=@n7, B=@n9), Exp=ScalarParameter() {defaultValue=4, parameterName=Exponent}), B=ParticleColor().RGB)
 ]
}
