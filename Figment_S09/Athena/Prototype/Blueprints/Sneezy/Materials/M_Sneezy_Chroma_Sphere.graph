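# Additive unlit material for mesh particles in the Sneezy prototype that draws a sphere with a chromatic-aberration-style rim glow.
# Three Fresnel terms sharing one exponent become red, green and blue; red and blue use pixel normals biased by view-transformed parameter vectors, splitting the channels.
# Particle color tints the summed result, which depth fade and near-camera fade then attenuate.
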Material /Figment_S09/Athena/Prototype/Blueprints/Sneezy/Materials/M_Sneezy_Chroma_Sphere
graph {
 properties {bUsedWithMeshParticles=true, blendMode=BLEND_Additive, pixelDepthOffsetMode=PDOM_Legacy, shadingModel=MSM_Unlit, usedShadingModels=Unlit}
 nodes [
  n1 = ParticleColor()
  n2 = ScalarParameter() {defaultValue=10, parameterName=Exp}
  n3 = PixelNormalWS()
  n4 = Transform(None=VectorParameter() {parameterName=RBiasScreenSpace, defaultValue=<rgba 0.2 0 0 0>}.RGB) {transformSourceType=TRANSFORMSOURCE_World, transformType=TRANSFORM_View}
  n5 = MaterialFunctionCall(Exponent=@n2, Vertex / Pixel Normal=Normalize(VectorInput=Add(A=@n3, B=@n4))) {materialFunction=&MF_Fresnel_FX}
  n6 = Constant3Vector() {constant=<rgba 1 0 0 0>}
  n7 = PixelNormalWS()
  n8 = Transform(None=VectorParameter() {parameterName=BBiasScreenSpace, defaultValue=<rgba -0.2 0 0 0>}.RGB) {transformSourceType=TRANSFORMSOURCE_World, transformType=TRANSFORM_View}
  n9 = MaterialFunctionCall(Exponent=@n2, Vertex / Pixel Normal=Normalize(VectorInput=Add(A=@n7, B=@n8))) {materialFunction=&MF_Fresnel_FX}
  n10 = Constant3Vector() {constant=<rgba 0 0 1 0>}
  n11 = Add(
   A=Multiply(
    A=MaterialFunctionCall(Exponent=@n2) {materialFunction=&MF_Fresnel_FX}.Fresnel
    B=Constant3Vector() {constant=<rgba 0 1 0 0>}
   )
   B=Add(A=Multiply(A=@n5.Fresnel, B=@n6), B=Multiply(A=@n9.Fresnel, B=@n10))
  )
  Multiply(
   A=DepthFade(Opacity=Multiply(A=@n1.RGB, B=@n11), FadeDistance=ScalarParameter() {defaultValue=500, parameterName=FadeDistance})
   B=MaterialFunctionCall(Distance=ScalarParameter() {defaultValue=100, parameterName=NearCameraFade Distance}) {materialFunction=&Near_Camera_Fade}.Result
  )
 ]
}
