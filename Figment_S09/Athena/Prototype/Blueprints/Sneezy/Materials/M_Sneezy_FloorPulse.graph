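# Unlit Modulate material for the Sneezy floor pulse prototype, drawing distorted chromatic lines that react to the Creative Jukebox audio band parameters.
# Two panning flow-map samples distort three gradient lookups packed as RGB, with Mid and Low scaling green and blue offsets for color fringing.
# The five-band average drives distortion and intensity, while an edge mask times FFTOpacity (default zero) fades to white, modulate's neutral.
# A per-band RGB tint branch appears unconnected.
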Material /Figment_S09/Athena/Prototype/Blueprints/Sneezy/Materials/M_Sneezy_FloorPulse
graph {
 properties {
  bEnableResponsiveAA=true
  blendMode=BLEND_Modulate
  pixelDepthOffsetMode=PDOM_Legacy
  shadingModel=MSM_Unlit
  translucencyPass=MTP_BeforeDOF
  usedShadingModels=Unlit
 }
 nodes [
  n1 = AppendVector(A=ScalarParameter() {defaultValue=1, parameterName=LinesTileScaleU}, B=ScalarParameter() {defaultValue=1.2, parameterName=LinesTileScaleV})
  n2 = Multiply(A=TextureCoordinate(), B=Multiply(A=ScalarParameter() {defaultValue=20, parameterName=Base Tile}, B=@n1))
  n3 = Add(
   A=CollectionParameter() {collection=/Game/Creative/Devices/Jukebox/Creative_Jukebox_MPC, parameterName=Low}
   B=CollectionParameter() {collection=&Creative_Jukebox_MPC, parameterName=LowMid}
  )
  n4 = CollectionParameter() {collection=&Creative_Jukebox_MPC, parameterName=Mid}
  n5 = Add(
   A=Add(A=Add(A=@n3, B=@n4), B=CollectionParameter() {collection=&Creative_Jukebox_MPC, parameterName=MidHigh})
   B=CollectionParameter() {collection=&Creative_Jukebox_MPC, parameterName=High}
  )
  n6 = ComponentMask(None=VertexInterpolator(VS=Divide(A=@n5) {constB=5}).PS) {r=true}
  n7 = TextureCoordinate()
  n8 = AppendVector(A=ScalarParameter() {defaultValue=-0.4, parameterName=LinesPanU}, B=ScalarParameter() {parameterName=LinesPanVv})
  n9 = Panner(
   Coordinate=Multiply(A=Multiply(A=@n7, B=@n1), B=ScalarParameter() {defaultValue=2, parameterName=DistortA-Tile})
   Speed=Multiply(A=@n8) {constB=0.3334}
  ) {bFractionalPart=true}
  n10 = Multiply(A=@n1, B=TextureCoordinate())
  n11 = ScalarParameter() {defaultValue=4, parameterName=DistortB-Tile}
  n12 = Panner(Coordinate=Add(A=ScalarParameter() {defaultValue=0.33, parameterName=DistortB-Offset}, B=Multiply(A=@n10, B=@n11)), Speed=@n8) {bFractionalPart=true}
  n13 = Add(
   A=ComponentMask(None=TextureSample(UVs=@n9) {samplerType=SAMPLERTYPE_LinearColor, texture=/Figment_S09/Environments/Sets/DigSite/Textures/T_Push_Flow}.RGB) {g=true, r=true}
   B=ComponentMask(None=TextureSample(UVs=@n12) {samplerType=SAMPLERTYPE_LinearColor, texture=&T_Push_Flow}.RGB) {g=true, r=true}
  )
  n14 = Multiply(
   A=LinearInterpolate(
    A=ScalarParameter() {defaultValue=4, parameterName=Distort-BaseAmtLow}
    B=ScalarParameter() {defaultValue=10, parameterName=Distort-BaseAmt}
    Alpha=@n6
   )
   B=Subtract(A=Normalize(VectorInput=@n13)) {constB=0.5}
  )
  n15 = Multiply(
   A=@n14
   B=LinearInterpolate(
    B=ScalarParameter() {defaultValue=1.06, parameterName=Distort-GreenAmt}
    Alpha=CollectionParameter() {collection=&Creative_Jukebox_MPC, parameterName=Mid}
   ) {constA=1}
  )
  n16 = Multiply(
   A=@n14
   B=LinearInterpolate(
    B=ScalarParameter() {defaultValue=1.06, parameterName=Distort-BlueAmt}
    Alpha=CollectionParameter() {collection=&Creative_Jukebox_MPC, parameterName=Low}
   ) {constA=1}
  )
  n17 = MaterialFunctionCall(
   X=TextureSample(UVs=Add(A=@n2, B=@n14)) {samplerType=SAMPLERTYPE_LinearColor, texture=/Game/Weapons/WeaponSkins/Textures/T_VertGradient}.R
   Y=TextureSample(UVs=Add(A=@n15, B=@n2)) {samplerType=SAMPLERTYPE_LinearColor, texture=&T_VertGradient}.R
   Z=TextureSample(UVs=Add(A=@n16, B=@n2)) {samplerType=SAMPLERTYPE_LinearColor, texture=&T_VertGradient}.R
  ) {materialFunction=&MakeFloat3}
  n18 = ScalarParameter() {defaultValue=-2, parameterName=Fraction}
  n19 = LinearInterpolate(
   A=VectorParameter() {parameterName=Low, defaultValue=<rgba 2 2 2 0>}.RGB
   B=VectorParameter() {parameterName=High, defaultValue=<rgba 3.5 3.5 3.5 0>}.RGB
   Alpha=Power(Base=Desaturation(None=@n17.Result, Fraction=@n18), Exp=ScalarParameter() {defaultValue=3, parameterName=ChromaPower})
  )
  n20 = LinearInterpolate(
   B=VectorParameter() {parameterName=Red, defaultValue=<rgba 0.25 0 0 0>}.RGB
   Alpha=CollectionParameter() {collection=&Creative_Jukebox_MPC, parameterName=Low}
  ) {constA=1}
  n21 = LinearInterpolate(
   B=VectorParameter() {parameterName=Green, defaultValue=<rgba 0 0.25 0 0>}.RGB
   Alpha=CollectionParameter() {collection=&Creative_Jukebox_MPC, parameterName=Mid}
  ) {constA=1}
  n22 = VectorParameter() {parameterName=Blue, defaultValue=<rgba 0 0 0.25 0>}
  n23 = CollectionParameter() {collection=&Creative_Jukebox_MPC, parameterName=High}
  Add(A=@n19, B=Add(A=Add(A=@n20, B=@n21), B=LinearInterpolate(B=@n22.B, Alpha=@n23) {constA=1}))
  ConstantBiasScale(None=Sine(None=Time()))
  n24 = TextureCoordinate()
  n25 = Power(Base=Abs(None=ConstantBiasScale(None=@n24) {bias=-0.5, scale=2}), Exp=ScalarParameter() {defaultValue=8, parameterName=EdgeMaskExp})
  n26 = Constant()
  n27 = Constant() {r=1}
  n28 = Constant() {r=0.2}
  LinearInterpolate(
   B=LinearInterpolate(B=@n19, Alpha=@n6) {constA=1}
   Alpha=Multiply(A=SphereMask(A=@n25, B=@n26, Radius=@n27, Hardness=@n28), B=ScalarParameter() {parameterName=FFTOpacity})
  ) {constA=1}
 ]
}
