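# Additive, two-sided unlit material for Sneezy magnetic energy ribbons, usable on mesh particles, that react to Jukebox audio.
# Panning, noise-distorted magnetic-line and chroma-ribbon textures combine, with a boosted tip, end fade and optional particle color.
# Averaged frequency bands from the Jukebox parameter collection, scaled by FFTOpacity, drive distortion and ribbon intensity.
# Vertex offset combines SplineThicken, swirl noise and an outward bend around DanceFloorCenterPoint; depth and distance fades soften it.
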Material /Figment_S09/Athena/Prototype/Blueprints/Sneezy/Materials/M_Sneezy_MagneticEnergy
graph {
 properties {
  bEnableResponsiveAA=true
  bUsedWithMeshParticles=true
  blendMode=BLEND_Additive
  pixelDepthOffsetMode=PDOM_Legacy
  shadingModel=MSM_Unlit
  twoSided=true
  usedShadingModels=Unlit
 }
 nodes [
  n1 = MaterialFunctionCall(Float2=TextureCoordinate()) {materialFunction=&BreakOutFloat2Components}
  n2 = SphereMask(
   A=@n1.G
   B=ScalarParameter() {defaultValue=1, parameterName=TipCenter}
   Radius=ScalarParameter() {defaultValue=0.3, parameterName=TipRadius}
   Hardness=ScalarParameter() {defaultValue=0.1, parameterName=TipHardness}
  )
  n3 = OneMinus(None=@n2)
  n4 = MaterialFunctionCall(Float2=TextureCoordinate()) {materialFunction=&BreakOutFloat2Components}
  n5 = OneMinus(None=@n4.G)
  n6 = ScalarParameter() {defaultValue=1.8, parameterName=U_Exp, sortPriority=1}
  n7 = Multiply(
   A=MaterialFunctionCall(X=@n4.R, Y=Power(Base=@n5, Exp=@n6)) {materialFunction=&MakeFloat2}.Result
   B=AppendVector(
    A=Constant() {r=1}
    B=MaterialFunctionCall() {materialFunction=/Engine/Functions/Engine_MaterialFunctions02/WorldPositionOffset/ObjectScale}.Scale X
   )
  )
  n8 = AppendVector(A=ScalarParameter() {defaultValue=1, parameterName=NoiseA_U_Tile}, B=ScalarParameter() {defaultValue=0.04, parameterName=NoiseA_V_Tile})
  n9 = ScalarParameter() {parameterName=NoiseA_U_PanSpeed}
  n10 = ScalarParameter() {defaultValue=0.1335, parameterName=NoiseA_V_PanSpeed}
  n11 = TextureSample(UVs=Panner(Coordinate=Multiply(A=@n7, B=@n8), Speed=AppendVector(A=@n9, B=@n10)) {bFractionalPart=true}) {samplerType=SAMPLERTYPE_Normal, texture=/Figment_S09/Environments/Foliage/Textures/T_SwirlNoise_N}
  n12 = MaterialFunctionCall(Float3=@n11.RGB) {materialFunction=&BreakOutFloat3Components}
  n13 = Multiply(
   A=MaterialFunctionCall(X=@n12.R, Y=@n12.R, Z=Subtract(A=@n12.B)) {materialFunction=&MakeFloat3}.Result
   B=VectorParameter() {parameterName=WPONoiseScale, defaultValue=<rgba 70 70 140 0>}.RGB
  )
  n14 = VectorParameter() {parameterName=DanceFloorCenterPoint, defaultValue=<rgba -1886 1532 16 0>}
  n15 = Subtract(A=WorldPosition().XYZ, B=@n14.RGB)
  n16 = WorldPosition()
  n17 = ScalarParameter() {defaultValue=3600, parameterName=BendRadius}
  n18 = ScalarParameter() {parameterName=BendHardness}
  n19 = Multiply(
   A=Multiply(A=Normalize(VectorInput=@n15), B=ScalarParameter() {defaultValue=1500, parameterName=BendAmount})
   B=Power(Base=SphereMask(A=@n14.RGB, B=@n16.XYZ, Radius=@n17, Hardness=@n18), Exp=ScalarParameter() {defaultValue=10, parameterName=BendExponent})
  )
  n20 = DynamicParameter() {defaultValue=<rgba 200 200 1 1>, paramNames [ThicknessStart, ThicknessEnd, none, none]}
  n21 = ScalarParameter() {defaultValue=10, parameterName=ThicknessStart}
  n22 = ScalarParameter() {defaultValue=200, parameterName=ThicknessEnd}
  n23 = WorldPosition()
  Add(
   A=Add(A=Multiply(A=@n3, B=@n13), B=@n19)
   B=MaterialFunctionCall(
    WidthBase=StaticSwitchParameter(True=@n20.ThicknessStart, False=@n21) {parameterName=UseDynamicParams}
    WidthTip=StaticSwitchParameter(True=@n20.ThicknessEnd, False=@n22) {parameterName=UseDynamicParams}
    WorldPosition=Add(A=@n19, B=@n23.XYZ)
   ) {materialFunction=&SplineThicken}.WorldPositionOffset
  )
  n24 = ParticleColor()
  n25 = ScalarParameter() {defaultValue=1, parameterName=EndFadeCenter}
  n26 = ScalarParameter() {defaultValue=0.95, parameterName=EndFadeRadius}
  n27 = ScalarParameter() {defaultValue=0.5, parameterName=EndFadeHardness}
  n28 = ComponentMask(None=@n11.RGB) {g=true, r=true}
  n29 = ScalarParameter() {defaultValue=0.025, parameterName=MagneticUVDistortionScale}
  n30 = ScalarParameter() {defaultValue=0.09, parameterName=MagneticUVDistortionScaleFFT}
  n31 = CollectionParameter() {collection=/Game/Creative/Devices/Jukebox/Creative_Jukebox_MPC, parameterName=Low}
  n32 = CollectionParameter() {collection=&Creative_Jukebox_MPC, parameterName=LowMid}
  n33 = Add(
   A=Add(A=Add(A=@n31, B=@n32), B=CollectionParameter() {collection=&Creative_Jukebox_MPC, parameterName=Mid})
   B=CollectionParameter() {collection=&Creative_Jukebox_MPC, parameterName=MidHigh}
  )
  n34 = CollectionParameter() {collection=&Creative_Jukebox_MPC, parameterName=High}
  n35 = Power(Base=Divide(A=Add(A=@n33, B=@n34)) {constB=5}, Exp=ScalarParameter() {defaultValue=1.85, parameterName=FftExp})
  n36 = Multiply(A=ComponentMask(None=VertexInterpolator(VS=@n35).PS) {r=true}, B=ScalarParameter() {parameterName=FFTOpacity})
  n37 = Add(A=@n7, B=Multiply(A=@n28, B=LinearInterpolate(A=@n29, B=@n30, Alpha=@n36)))
  n38 = AppendVector(A=ScalarParameter() {defaultValue=1, parameterName=MagneticU_Tile}, B=ScalarParameter() {defaultValue=0.125, parameterName=MagneticV_Tile})
  n39 = ScalarParameter() {parameterName=MagneticU_PanSpeed}
  n40 = TextureCoordinate() {uTiling=8}
  n41 = ScalarParameter() {defaultValue=1.4, parameterName=RandomSeed}
  n42 = Multiply(
   A=ScalarParameter() {defaultValue=0.8316, parameterName=MagneticV_PanSpeed}
   B=ConstantBiasScale(None=MaterialFunctionCall!(PrimaryTextureCoordinates=@n40, 2ndOrderTextureCoordinates, RandomSeed=@n41, NoiseDensityBias, Bypass Noise Influence Crosstalk) {materialFunction=/Game/Effects/Fort_Effects/Materials/_Functions/MF_RandomUVSample}.NoiseWithRange) {bias=0.5}
  )
  n43 = TextureSample(UVs=Panner(Coordinate=Multiply(A=@n37, B=@n38), Speed=AppendVector(A=@n39, B=@n42)) {bFractionalPart=true}) {constMipValue=0, texture=/Game/Athena/Prototype/Blueprints/Sneezy/Textures/T_Sneezy_MagneticWaves}
  n44 = ScalarParameter() {defaultValue=2, parameterName=MagneticLinesExponent}
  n45 = Add(A=@n7, B=Multiply(A=@n28, B=ScalarParameter() {defaultValue=0.15, parameterName=ChromaUVDistortionScale}))
  n46 = AppendVector(
   A=ScalarParameter() {defaultValue=1, parameterName=ChromaRibbonU_Tile}
   B=ScalarParameter() {defaultValue=0.03, parameterName=ChromaRibbonV_Tile}
  )
  n47 = ScalarParameter() {parameterName=ChromaRibbonU_PanSpeed}
  n48 = ScalarParameter() {defaultValue=0.06, parameterName=ChromaRibbonV_PanSpeed}
  n49 = TextureSample(UVs=Panner(Coordinate=Multiply(A=@n45, B=@n46), Speed=AppendVector(A=@n47, B=@n48)) {bFractionalPart=true}) {texture=/Game/Athena/Prototype/Blueprints/Sneezy/Textures/T_Sneezy_EnergyWaves}
  n50 = ScalarParameter() {defaultValue=1.3, parameterName=ChromaRibbonPower}
  n51 = ScalarParameter() {defaultValue=0.15, parameterName=ChromaRibbonScale}
  n52 = ScalarParameter() {defaultValue=1, parameterName=ChromaRibbonScaleFFT}
  n53 = Add(
   A=Multiply(A=Power(Base=@n43.RGB, Exp=@n44), B=ScalarParameter() {defaultValue=1.5, parameterName=MagneticLinesScale})
   B=Multiply(A=Power(Base=@n49.RGB, Exp=@n50), B=LinearInterpolate(A=@n51, B=@n52, Alpha=@n36))
  )
  n54 = Multiply(A=@n53, B=ScalarParameter() {defaultValue=20, parameterName=TipScale})
  n55 = Multiply(
   A=Multiply(A=SphereMask(A=@n1.G, B=@n25, Radius=@n26, Hardness=@n27), B=LinearInterpolate(A=@n53, B=@n54, Alpha=@n2))
   B=ScalarParameter() {defaultValue=1, parameterName=Opacity}
  )
  StaticSwitchParameter(True=Multiply(A=@n24.RGB, B=@n55), False=@n55) {parameterName=UseParticleColor}
  n56 = Constant() {r=1}
  n57 = DepthFade(FadeDistance=ScalarParameter() {defaultValue=20, parameterName=FadeDistance})
  n58 = MaterialFunctionCall(
   Opacity In=QualitySwitch(Default=@n56, Low=@n56, High=@n57, Medium=@n57)
   Position=WorldPosition().XYZ
   Fade_Start_Distance=ScalarParameter() {defaultValue=200, parameterName="Fade_Start_Distance (S)"}
   Fade_End_Distance=ScalarParameter() {defaultValue=100, parameterName="Fade_End_Distance (S)"}
  ) {materialFunction=&MF_Distance_SyncedFade_FX}
  StaticSwitchParameter(True=Multiply(A=@n24.A, B=@n58.Result), False=@n58.Result) {parameterName=UseParticleColor}
  Time() {bOverride_Period=true, period=1}
 ]
}
