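# Translucent unlit material for the Sneezy prototype's camera-facing radial meter, driven by the NormalizedLevel parameter.
# The level rotates radial-gradient UVs that set the fill, hot edge and needle, with T_Sneezy_Meter_01 channels masking regions.
# Higher levels also shift outline and needle colors, strengthen the glow and add a sine-timed "dance" pulse through world position offset.
# Animated wrap patterns break up the fill, and flow-distorted scrolling gradients, offset per RGB channel, add chromatic lines.
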Material /Figment_S09/Athena/Prototype/Blueprints/Sneezy/Materials/M_Sneezy_Meter
graph {
 properties {
  bEnableResponsiveAA=true
  bUsedWithMeshParticles=true
  blendMode=BLEND_Translucent
  pixelDepthOffsetMode=PDOM_Legacy
  shadingModel=MSM_Unlit
  usedShadingModels=Unlit
 }
 nodes [
  n1 = Multiply(A=SphereMask(A=TextureCoordinate(), B=Constant() {r=0.5}, Radius=Constant() {r=0.5}, Hardness=Constant() {r=0.75})) {constB=0.1}
  n2 = ComponentMask(None=TextureCoordinate()) {g=true}
  Add(B=Multiply(A=Power(Base=@n1) {constExponent=1.4}, B=Power(Base=@n2)))
  n3 = Subtract(A=ObjectPositionWS(), B=WorldPosition().XYZ)
  n4 = ScalarParameter() {parameterName=NormalizedLevel}
  n5 = Divide(A=Multiply(A=ScalarParameter() {defaultValue=4.5, parameterName=DancePulseIntensity}, B=@n4), B=ObjectRadius())
  n6 = Sine(None=Time()) {period=0.2}
  n7 = Saturate(None=MaterialFunctionCall(Input=@n4, Input Low=Constant(), Input High=Constant() {r=0.8}, Target Low=Constant() {r=-1.2}, Target High=Constant() {r=1}) {materialFunction=&RemapValueRange}.Result)
  Add(
   A=Multiply(A=Multiply(A=@n3, B=@n5), B=Multiply(A=@n6, B=@n7))
   B=MaterialFunctionCall!(
    WorldPosition=WorldPosition().XYZ
    PivotLocation
    WS Normals
    Custom Object Basis 1=Constant3Vector() {constant=<rgba 0 1 0 1>}
    Custom Object Basis 2=Constant3Vector() {constant=<rgba 0 0 -1 0>}
    Custom Object Basis 3
   ) {materialFunction=/Engine/Functions/Engine_MaterialFunctions02/WorldPositionOffset/AlignMeshToTheCamera}.World Position Offset
  )
  n8 = TextureSample() {samplerType=SAMPLERTYPE_LinearColor, texture=/Figment_S09/Athena/Prototype/Blueprints/Sneezy/Textures/T_Sneezy_Meter_01}
  n9 = LinearInterpolate(
   A=VectorParameter() {parameterName=BGCircleColor, defaultValue=<rgba 0.0967 0 1 1>}.RGB
   B=LinearInterpolate(
    A=VectorParameter() {parameterName=OutlineColorA, defaultValue=<rgba 0.1635 0.4858 0.6 1>}.RGB
    B=VectorParameter() {parameterName=OutlineColorB, defaultValue=<rgba 1.5137 0.4016 8 1>}.RGB
    Alpha=@n7
   )
   Alpha=@n8.A
  )
  n10 = Multiply(
   A=AppendVector(A=ScalarParameter() {defaultValue=4, parameterName=PatternUTile}, B=ScalarParameter() {defaultValue=2, parameterName=PatternVTile})
   B=TextureCoordinate()
  )
  n11 = Add(A=TextureSample(UVs=@n10) {texture=/Figment_S09/Weapons/WeaponSkins/Wraps/Textures/Snow_Board/T_Wrap_Snow_Board_00}.G, B=Time())
  n12 = ConstantBiasScale(None=Multiply(A=Sine(None=@n11) {period=8}, B=Cosine(None=@n11) {period=2})) {scale=0.2}
  n13 = Multiply(
   A=TextureCoordinate()
   B=AppendVector(A=ScalarParameter() {defaultValue=2, parameterName=UScale}, B=ScalarParameter() {defaultValue=2, parameterName=VScale})
  )
  n14 = AppendVector(A=ScalarParameter() {parameterName=UBias}, B=ScalarParameter() {defaultValue=-1.8, parameterName=VBias})
  n15 = ScalarParameter() {defaultValue=-9.5, parameterName=UVRotationStart}
  n16 = ScalarParameter() {defaultValue=-13.65, parameterName=UVRotationEnd}
  n17 = Subtract(A=Rotator(Coordinate=Add(A=@n13, B=@n14), Time=LinearInterpolate(A=@n15, B=@n16, Alpha=@n4)) {centerX=1, centerY=1})
  n18 = MaterialFunctionCall!(
   min=ScalarParameter() {parameterName=GradientStartInverseLerp}
   max=ScalarParameter() {defaultValue=0.29, parameterName=GradientEndInverseLerp}
   value=ComponentMask(None=MaterialFunctionCall(Vector or UVs=@n17) {materialFunction=&VectorToRadialValue}.Radial Coordinates) {r=true}
  ) {materialFunction=/Game/Effects/Fort_Effects/Materials/_Functions/InverseLerp_MF}
  n19 = LinearInterpolate(
   A=VectorParameter() {parameterName=MeterColorA, defaultValue=<rgba 1 0.0645 0.6533 0>}.RGB
   B=VectorParameter() {parameterName=MeterColorB, defaultValue=<rgba 0.0572 0.712 2 0>}.RGB
   Alpha=LinearInterpolate(B=Add(A=@n12, B=@n18.Clamped), Alpha=@n18.Clamped)
  )
  n20 = VectorParameter() {parameterName=HotColor, defaultValue=<rgba 0.3093 3 2.3874 0>}
  n21 = ConstantBiasScale(None=@n6) {bias=1.5}
  n22 = Sine(None=Ceil(None=@n18.Unclamped)) {period=4}
  n23 = Multiply(
   A=@n22
   B=SphereMask(
    A=@n18.Unclamped
    B=Constant() {r=1}
    Radius=ScalarParameter() {defaultValue=0.175, parameterName=HotRadius}
    Hardness=ScalarParameter() {defaultValue=0.2, parameterName=HotHardness}
   )
  )
  n24 = Multiply(A=Power(Base=Multiply(A=@n8.B, B=@n23), Exp=ScalarParameter() {defaultValue=3, parameterName=HotExponent}), B=@n7)
  n25 = LinearInterpolate(A=@n19, B=Multiply(A=@n19, B=Multiply(A=@n20.RGB, B=@n21)), Alpha=@n24)
  n26 = VectorParameter() {parameterName=NeedleColorA, defaultValue=<rgba 0.0122 0.0784 1 0>}
  n27 = VectorParameter() {parameterName=NeedleColorB, defaultValue=<rgba 25.8566 9.352 50 0>}
  n28 = Constant() {r=1}
  n29 = ScalarParameter() {defaultValue=0.01, parameterName=NeedleRadius}
  n30 = ScalarParameter() {defaultValue=0.1, parameterName=NeedleHardness}
  n31 = Multiply(
   A=Add(A=@n8.A, B=@n8.A)
   B=Power(Base=SphereMask(A=@n18.Unclamped, B=@n28, Radius=@n29, Hardness=@n30), Exp=ScalarParameter() {defaultValue=4, parameterName=NeedleExponent})
  )
  n32 = ScalarParameter() {defaultValue=1, parameterName=ChromaScale}
  n33 = AppendVector(A=ScalarParameter() {defaultValue=1, parameterName=LinesTileScaleU}, B=ScalarParameter() {defaultValue=0.6, parameterName=LinesTileScaleV})
  n34 = Multiply(A=TextureCoordinate(), B=Multiply(A=ScalarParameter() {defaultValue=12, parameterName=Base Tile}, B=@n33))
  n35 = TextureCoordinate()
  n36 = Time()
  n37 = ScalarParameter() {defaultValue=0.05, parameterName=DistortA-Speed}
  n38 = Add(
   A=Multiply(A=Multiply(A=@n35, B=@n33), B=ScalarParameter() {defaultValue=2, parameterName=DistortA-Tile})
   B=Multiply(A=Multiply(A=@n36, B=@n37), B=Constant2Vector() {g=0.25, r=1})
  )
  n39 = Multiply(A=@n33, B=TextureCoordinate())
  n40 = ScalarParameter() {defaultValue=4, parameterName=DistortB-Tile}
  n41 = Multiply(A=Time(), B=ScalarParameter() {defaultValue=-0.1, parameterName=DistortB-Speed})
  n42 = Constant2Vector() {g=1, r=0.25}
  n43 = Add(A=ScalarParameter() {defaultValue=0.33, parameterName=DistortB-Offset}, B=Add(A=Multiply(A=@n39, B=@n40), B=Multiply(A=@n41, B=@n42)))
  n44 = Add(
   A=ComponentMask(None=TextureSample(UVs=@n38) {samplerType=SAMPLERTYPE_LinearColor, texture=/Figment_S09/Environments/Sets/DigSite/Textures/T_Push_Flow}.RGB) {g=true, r=true}
   B=ComponentMask(None=TextureSample(UVs=@n43) {samplerType=SAMPLERTYPE_LinearColor, texture=&T_Push_Flow}.RGB) {g=true, r=true}
  )
  n45 = Multiply(A=ScalarParameter() {defaultValue=12, parameterName=Distort-BaseAmt}, B=Subtract(A=Normalize(VectorInput=@n44)) {constB=0.5})
  n46 = Add(A=@n34, B=@n45)
  n47 = Add(A=Multiply(A=@n45, B=ScalarParameter() {defaultValue=1.02, parameterName=Distort-GreenAmt}), B=@n34)
  n48 = Add(A=Multiply(A=@n45, B=ScalarParameter() {defaultValue=1.04, parameterName=Distort-BlueAmt}), B=@n34)
  n49 = MaterialFunctionCall(
   X=TextureSample(UVs=Panner(Coordinate=@n46) {speedY=-0.125}) {samplerType=SAMPLERTYPE_LinearColor, texture=/Game/Weapons/WeaponSkins/Textures/T_VertGradient}.R
   Y=TextureSample(UVs=Panner(Coordinate=@n47) {speedY=-0.125}) {samplerType=SAMPLERTYPE_LinearColor, texture=&T_VertGradient}.R
   Z=TextureSample(UVs=Panner(Coordinate=@n48) {speedY=-0.125}) {samplerType=SAMPLERTYPE_LinearColor, texture=&T_VertGradient}.R
  ) {materialFunction=&MakeFloat3}
  Add(
   A=LinearInterpolate(A=LinearInterpolate(A=@n9, B=@n25, Alpha=@n8.B), B=LinearInterpolate(A=@n26.RGB, B=@n27.RGB, Alpha=@n4), Alpha=@n31)
   B=Multiply(A=@n8.R, B=Multiply(A=@n32, B=@n49.Result))
  )
  n50 = OneMinus(None=@n8.G)
  n51 = ScalarParameter() {defaultValue=0.25, parameterName=MeterOpacity}
  n52 = Constant() {r=1}
  n53 = Multiply(A=SphereMask(A=@n18.Clamped, B=@n52, Radius=@n52, Hardness=Constant()), B=@n22)
  n54 = ScalarParameter() {defaultValue=1, parameterName=PatternMin}
  n55 = ScalarParameter() {defaultValue=0.8, parameterName=PatternMax}
  n56 = TextureSample(UVs=@n10) {texture=/Figment_S09/Weapons/WeaponSkins/Wraps/Textures/Snow_Board/T_Wrap_Snow_Board_03}
  n57 = LinearInterpolate(A=@n8.A, B=Multiply(A=Multiply(A=@n51, B=@n53), B=LinearInterpolate(A=@n54, B=@n55, Alpha=@n56.R)), Alpha=@n8.A)
  Add(A=Add(A=Multiply(A=@n50, B=@n57), B=@n31), B=@n24)
  Multiply(A=Time() {bOverride_Period=true, period=10}, B=Divide() {constB=10})
  OneMinus(None=@n2)
 ]
}
