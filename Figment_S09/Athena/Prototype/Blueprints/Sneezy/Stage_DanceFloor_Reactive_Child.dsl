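; Reactive dance floor stage piece, a child Blueprint that lights up while players overlap it and cannot be destroyed, as ShouldDie always returns false.
; On the server, a player pawn entering sets isActive, and an end overlap clears it once no Athena player pawns remain.
; OnRep_isActive swaps the mesh material between lit and off dance floor looks on remote machines only.
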(event EventBeginPlay
  (|CallParentFunction)
  (Variables|Default|SetisActive false))

(event Collision|EventActorBeginOverlap
  (params (OtherActor :type Actor%20Object%20Reference))
  (|CallParentFunction OtherActor)
  (Networking|SwitchHasAuthority
    (:Authority
      (Utilities|Casting|CastToFortPlayerPawnAthena OtherActor)
      (if (not (Variables|Default|GetisActive))
        (Variables|Default|SetisActive true)))))

(event Collision|EventActorEndOverlap
  (params (OtherActor :type Actor%20Object%20Reference))
  (|CallParentFunction OtherActor)
  (Networking|SwitchHasAuthority
    (:Authority
      (if (< (Utilities|Array|Length (Collision|GetOverlappingActors "/Game/Athena/PlayerPawn_Athena.PlayerPawn_Athena_C")) 1)
        (if (Variables|Default|GetisActive)
          (Variables|Default|SetisActive false))))))

(event EventTick
  (params (DeltaSeconds :type Float%20%28single-precision%29)))

(fn OnRep_isActive
  (Networking|SwitchHasAuthority
    (:Remote
      (if (Variables|Default|GetisActive)
        (Rendering|Material|SetMaterial (Variables|Debug|GetStaticMeshComponent) 0 "/Game/Packages/Fortress_SharedMaterials/Base_Material/DanceFloor.DanceFloor")
        (else
          (Rendering|Material|SetMaterial (Variables|Debug|GetStaticMeshComponent) 0 "/Game/Athena/Items/QuestInteractables/DancePlatforms/FX/MI_DanceFloorOffQuest.MI_DanceFloorOffQuest"))))))

(fn ShouldDie
  (params (Damage :type Float%20%28single-precision%29) (EventInstigator :type Controller%20Object%20Reference) (DamageCauser :type Actor%20Object%20Reference) (bIsTest :type Boolean) (EffectContext :type Gameplay%20Effect%20Context%20Handle%20Structure%20%28by%20ref%29))
  (return false))

(fn ConstructionScript
  (|CallParentFunction))
